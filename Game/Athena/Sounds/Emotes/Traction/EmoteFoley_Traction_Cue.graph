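# Foley sound cue for the Traction emote, playing engine and grass-cutting sounds that respond to movement speed.
# A one-shot start sound plays alongside looping layers that an enveloper fades in along an accelerating curve over 1.5 seconds.
# The EmoteMoveSpeed parameter crossfades the idle engine loop down as the high engine and grass-cut loops rise.
# It uses emote music class and concurrency, emote attenuation, play-when-silent virtualization and a music-analysis submix send.
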SoundCue /Game/Athena/Sounds/Emotes/Traction/EmoteFoley_Traction_Cue
root=Mixer_1
nodes [
 {
  id=Enveloper_0
  children [Mixer_0]
  properties {
   pitchCurve {editorCurveData {keys [{time=0}]}}
   volumeCurve {editorCurveData {keys [<at 0 0>,<at 0.4245 0.061>,<at 0.8708 0.1294>,<at 1.2115 0.2908>,<at 1.4179 0.6281>,<at 1.5 1>]}}
  }
 }
 {id=Mixer_0, children [ModulatorContinuous_1, ModulatorContinuous_0], properties {inputVolume [0.9,0.65]}}
 {id=Mixer_1, children [WavePlayer_3,Enveloper_0], properties {inputVolume [1,1]}}
 {id=Mixer_2, children [WavePlayer_4,WavePlayer_1], properties {inputVolume [1,2]}}
 {
  id=ModulatorContinuous_0
  children [Mixer_2]
  properties {pitchModulationParams {minOutput=1, parameterName=EmoteMoveSpeed}, volumeModulationParams {parameterName=EmoteMoveSpeed}}
 }
 {
  id=ModulatorContinuous_1
  children [WavePlayer_0]
  properties {
   pitchModulationParams {minOutput=1, parameterName=EmoteMoveSpeed}
   volumeModulationParams {maxOutput=0.2, minOutput=1, parameterName=EmoteMoveSpeed}
  }
 }
 {id=WavePlayer_0, properties {bLooping=true, soundWaveAssetPtr=&/Emote_Traction_Eng_Idle_LP}}
 {id=WavePlayer_1, properties {bLooping=true, soundWaveAssetPtr=&/Emote_Traction_GrassCut_LP}}
 {id=WavePlayer_3, properties {soundWaveAssetPtr=&/Emote_Traction_Start}}
 {id=WavePlayer_4, properties {bLooping=true, soundWaveAssetPtr=&/Emote_Traction_Eng_Hi_LP}}
]
properties {
 attenuationSettings=&Fort_Emotes_Athena_Att
 maxDistance=1600
 soundClassObject=&Fort_Music_Emote
 virtualizationMode=PlayWhenSilent
 volumeMultiplier=1
 concurrencyOverrides {maxCount=3}
 concurrencySet [/Game/Sounds/Concurrency_Rules/Athena_EmoteMusic_Conc]
 soundSubmixSends [{minSendLevel=0, soundSubmix=&Fortnite_AudioAnalysis_Music}]
}
